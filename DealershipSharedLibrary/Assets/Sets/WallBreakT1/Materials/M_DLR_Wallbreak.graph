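# Unlit masked material for Niagara particles that animates the pieces of a wall-break effect around a BallPosition point.
# Per-element pivots and random values packed in UV1 and UV2, plus distance to the ball, build staggered falloff masks.
# These masks and Niagara dynamic parameters drive transition pull, scale collapse, spin and orbit offsets using RotateAboutAxis.
# Emissive blends a base color with a masked custom color, brightened by vertex-red outlines.
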Material /DealershipSharedLibrary/Assets/Sets/WallBreakT1/Materials/M_DLR_Wallbreak
graph {
 properties {
  bAutomaticallySetUsageInEditor=false
  bUsedWithNiagaraMeshParticles=true
  bUsedWithNiagaraSprites=true
  blendMode=BLEND_Masked
  shadingModel=MSM_Unlit
  usedShadingModels=Unlit
 }
 nodes [
  n1 = NamedRerouteUsage()
  n2 = NamedRerouteUsage()
  Add(A=Add(A=Add(A=@n1, B=@n2), B=NamedRerouteUsage()), B=NamedRerouteUsage())
  Constant() {r=2}
  MaterialFunctionCall(Emissive=NamedRerouteUsage()) {materialFunction=&MF_FX_EmissiveAdjustments}
  n3 = TextureCoordinate() {coordinateIndex=2}
  n4 = Transform(None=AppendVector(A=TextureCoordinate() {coordinateIndex=1}, B=ComponentMask(None=@n3) {r=true})) {transformSourceType=TRANSFORMSOURCE_World, transformType=TRANSFORM_Local}
  n5 = ParticlePositionWS()
  n6 = VectorParameter() {parameterName=Offset, defaultValue=<rgba 0 0 320 1>}
  n7 = ComponentMask(None=@n6.RGB) {g=true, r=true}
  n8 = ComponentMask(None=WorldPosition().XYZ) {b=true}
  n9 = AppendVector(A=LinearInterpolate(A=@n7, B=OneMinus(None=@n7), Alpha=Clamp(None=@n8)), B=@n6.B)
  n10 = TransformPosition(None=VectorParameter() {parameterName=BallPosition, defaultValue=<rgba 0 0 320 1>}.RGB) {transformSourceType=TRANSFORMPOSSOURCE_Instance, transformType=TRANSFORMPOSSOURCE_World}
  NamedRerouteDeclaration(None=Distance(A=Add(A=@n4, B=@n5), B=Add(A=@n9, B=@n10))) {name=Distance_To_CenterBall, nodeColor=<rgba 0.1354 0.5542 1 1>}
  n11 = Multiply(A=ComponentMask(None=TextureCoordinate() {coordinateIndex=2}) {g=true}, B=ScalarParameter() {defaultValue=100, parameterName=Rand})
  n12 = Subtract(
   A=NamedRerouteUsage()
   B=LinearInterpolate(
    A=ScalarParameter() {defaultValue=700, parameterName=IntroMin}
    B=ScalarParameter() {defaultValue=1500, parameterName=IntroMax}
    Alpha=NamedRerouteUsage()
   )
  )
  n13 = OneMinus(None=Subtract(A=Add(A=@n11, B=@n12), B=ScalarParameter() {defaultValue=-650, parameterName=MaskOffset}))
  n14 = ScalarParameter() {defaultValue=0.008, parameterName=MaskFalloff}
  NamedRerouteDeclaration(None=Clamp(None=Multiply(A=@n13, B=@n14))) {name=Mask, nodeColor=<rgba 0.4833 0.1146 1 1>}
  n15 = DynamicParameter() {defaultValue=<rgba 1 0 -0.5 1>, paramNames [Transition, EndTransition, EndScale, Param4]}
  NamedRerouteDeclaration(None=@n15.EndTransition) {name=DP_EndTransition, nodeColor=<rgba 1 0.2083 0.2083 0.5>}
  NamedRerouteDeclaration(None=@n15.EndScale) {name=DP_EndScale, nodeColor=<rgba 1 0.2083 0.2083 0.5>}
  n16 = NamedRerouteUsage()
  n17 = NamedRerouteUsage()
  n18 = WorldPosition()
  n19 = Subtract(
   A=NamedRerouteUsage()
   B=LinearInterpolate(
    A=ScalarParameter() {defaultValue=-640, parameterName=EndTransitionMin}
    B=ScalarParameter() {defaultValue=800, parameterName=EndTransitionMax}
    Alpha=NamedRerouteUsage()
   )
  )
  n20 = ScalarParameter() {defaultValue=0.001, parameterName=EndSuckGradFalloff}
  n21 = Clamp(None=OneMinus(None=Multiply(A=@n19, B=@n20)))
  n22 = Multiply(A=Multiply(A=Clamp(None=@n16), B=Subtract(A=@n17, B=@n18.XYZ)), B=@n21)
  n23 = NamedRerouteUsage()
  n24 = OneMinus(None=NamedRerouteUsage())
  n25 = ScalarParameter() {defaultValue=-330, parameterName=ScaleMaskOffset}
  n26 = Clamp(None=Multiply(A=Subtract(A=@n24, B=@n25), B=ScalarParameter() {defaultValue=0.003, parameterName=ScaleMaskFallof}))
  n27 = NamedRerouteUsage()
  n28 = Clamp(None=Add(A=@n26, B=OneMinus(None=@n27)))
  n29 = Subtract(A=NamedRerouteUsage(), B=WorldPosition().XYZ)
  NamedRerouteDeclaration(None=Add(A=Multiply(A=@n22, B=@n23), B=Multiply(A=@n28, B=@n29))) {name=Scale, nodeColor=<rgba 0.1094 1 0.6034 1>}
  n30 = TransformPosition(None=VectorParameter() {parameterName=BallPosition, defaultValue=<rgba 0 0 320 1>}.RGB) {transformSourceType=TRANSFORMPOSSOURCE_Instance, transformType=TRANSFORMPOSSOURCE_World}
  n31 = NamedRerouteUsage()
  NamedRerouteDeclaration(None=Multiply(A=Subtract(A=@n30, B=@n31), B=@n21)) {name=Transition, nodeColor=<rgba 1 0.3341 0.1302 1>}
  n32 = Normalize(VectorInput=Transform(None=VectorParameter() {parameterName=EndOrbitRotationAxis, defaultValue=<rgba 0 1 1 1>}.RGB) {transformSourceType=TRANSFORMSOURCE_Local})
  n33 = ScalarParameter() {defaultValue=0.001, parameterName=EndOrbitFalloff}
  n34 = Clamp(None=OneMinus(None=Multiply(A=@n19, B=@n33)))
  n35 = ScalarParameter() {defaultValue=0.5, parameterName=Pi}
  n36 = Clamp(None=Multiply(A=Frac(None=@n34), B=MaterialFunctionCall(Multiplier=@n35) {materialFunction=&Pi}.Result)) {maxDefault=200}
  n37 = TransformPosition(None=VectorParameter() {parameterName=BallPosition, defaultValue=<rgba 0 0 320 1>}.RGB) {transformSourceType=TRANSFORMPOSSOURCE_Instance, transformType=TRANSFORMPOSSOURCE_World}
  n38 = NamedRerouteUsage()
  n39 = Abs(None=Sine(None=@n38) {period=1.3})
  n40 = Abs(None=Sine(None=@n38) {period=2})
  n41 = Abs(None=Sine(None=@n38))
  n42 = Frac(None=Multiply(A=@n21, B=ScalarParameter() {defaultValue=6, parameterName=EndSmallRot}))
  n43 = MaterialFunctionCall(Multiplier=Constant() {r=2}) {materialFunction=&Pi}
  n44 = RotateAboutAxis(
   NormalizedRotationAxis=Normalize(VectorInput=MaterialFunctionCall(R=@n39, G=@n40, B=@n41) {materialFunction=&AppendMany}.RGB)
   RotationAngle=Clamp(None=Multiply(A=@n42, B=@n43.Result)) {maxDefault=200}
   PivotPoint=NamedRerouteUsage()
   Position=Add(A=WorldPosition().XYZ, B=NamedRerouteUsage())
  )
  n45 = NamedRerouteUsage()
  n46 = Sine(None=@n45) {period=1.3}
  n47 = Sine(None=@n45) {period=2}
  n48 = Sine(None=@n45)
  n49 = Normalize(VectorInput=MaterialFunctionCall(R=Abs(None=@n46), G=Abs(None=@n47), B=Abs(None=@n48)) {materialFunction=&AppendMany}.RGB)
  n50 = Time()
  n51 = NamedRerouteUsage()
  n52 = Ceil(None=Subtract(A=@n51, B=Constant() {r=0.5}))
  n53 = Add(A=LinearInterpolate(A=@n50, B=OneMinus(None=@n50), Alpha=Clamp(None=@n52)), B=@n51)
  n54 = ScalarParameter() {defaultValue=0.1, parameterName=SmallRotationSpeed}
  n55 = Add(A=WorldPosition().XYZ, B=NamedRerouteUsage())
  n56 = RotateAboutAxis(
   NormalizedRotationAxis=@n49
   RotationAngle=Multiply(A=Multiply(A=@n53, B=@n54), B=Ceil(None=@n26))
   PivotPoint=NamedRerouteUsage()
   Position=Add(A=@n44, B=@n55)
  )
  n57 = Add(A=NamedRerouteUsage(), B=Add(A=@n56, B=@n55))
  n58 = Add(A=@n44, B=@n57)
  NamedRerouteDeclaration(None=Add(A=RotateAboutAxis(NormalizedRotationAxis=@n32, RotationAngle=@n36, PivotPoint=@n37, Position=@n58), B=@n44)) {name=Rotation, nodeColor=<rgba 0.1042 0.1812 1 1>}
  n59 = Subtract(A=OneMinus(None=NamedRerouteUsage()), B=ScalarParameter() {defaultValue=-330, parameterName=IntroOrbitOffset})
  n60 = ScalarParameter() {defaultValue=0.03, parameterName=IntroOrbitFalloff}
  n61 = Multiply(A=Clamp(None=Multiply(A=@n59, B=@n60)), B=Multiply(A=@n53, B=ScalarParameter() {defaultValue=0.4, parameterName=OrbitSpeed}))
  n62 = TransformPosition(None=VectorParameter() {parameterName=BallPosition, defaultValue=<rgba 0 0 320 1>}.RGB) {transformType=TRANSFORMPOSSOURCE_World}
  n63 = Add(A=NamedRerouteUsage(), B=@n57)
  NamedRerouteDeclaration(None=Add(A=RotateAboutAxis(NormalizedRotationAxis=@n49, RotationAngle=@n61, PivotPoint=@n62, Position=@n63), B=@n56)) {name=CenterOrbit, nodeColor=<rgba 0.5295 1 0.151 1>}
  NamedRerouteDeclaration(None=@n15.Transition) {name=DP_Transition, nodeColor=<rgba 1 0.2083 0.2083 0.5>}
  n64 = AppendVector(A=TextureCoordinate() {coordinateIndex=1}, B=ComponentMask(None=TextureCoordinate() {coordinateIndex=2}) {r=true})
  n65 = Constant3Vector() {constant=<rgba 0 -1 0 0>}
  NamedRerouteDeclaration(None=TransformPosition(None=Add(A=@n64, B=@n65)) {transformSourceType=TRANSFORMPOSSOURCE_Instance, transformType=TRANSFORMPOSSOURCE_World}) {name=PivotPosition, nodeColor=<rgba 0.1302 0.9419 1 1>}
  NamedRerouteDeclaration(None=ComponentMask(None=TextureCoordinate() {coordinateIndex=2}) {g=true}) {name=RandomValuPerElement, nodeColor=<rgba 0.8047 1 0 1>}
  n66 = Multiply(A=Constant(), B=ReflectionVectorWS())
  n67 = Clamp(None=Add(A=ScalarParameter() {defaultValue=-0.7638, parameterName=MaskColorOffset}, B=NamedRerouteUsage()))
  n68 = Multiply(A=@n67, B=NamedRerouteUsage())
  n69 = ScalarParameter() {defaultValue=1, parameterName=RandomColor}
  n70 = NamedRerouteUsage()
  n71 = Multiply(A=Subtract(A=ScalarParameter() {defaultValue=-512, parameterName=CenterColorMaskOffset}, B=OneMinus(None=@n70)), B=Constant() {r=0.005})
  n72 = Multiply(
   A=VectorParameter() {parameterName=CustomColor, defaultValue=<rgba 1 0.125 0.3 1>}.RGB
   B=ScalarParameter() {defaultValue=1, parameterName=CustomColor_Intensity}
  )
  n73 = Multiply(A=VertexColor().R, B=ScalarParameter() {defaultValue=12, parameterName=OutLineBrightness})
  n74 = ScalarParameter() {defaultValue=1, parameterName=RingColor_Intensity}
  n75 = LinearInterpolate(
   A=Constant()
   B=Add(A=Multiply(A=@n72, B=@n73), B=Multiply(A=@n72, B=@n74))
   Alpha=Multiply(A=LinearInterpolate(A=@n67, B=@n68, Alpha=@n69), B=Clamp(None=@n71))
  )
  n76 = VectorParameter() {parameterName=BaseColor, defaultValue=<rgba 0.017 0.005 0.32 0>}
  n77 = ScalarParameter() {defaultValue=0.15, parameterName=BaseColor_Intensity}
  NamedRerouteDeclaration(None=Add(A=Add(A=@n66, B=@n75), B=Multiply(A=@n76.RGB, B=@n77))) {name=EmissiveColor, nodeColor=<rgba 0 0.8125 1 1>}
  NamedRerouteUsage()
  ObjectPositionWS()
  ScalarParameter() {defaultValue=1, parameterName=EndScale}
  ScalarParameter() {defaultValue=1, parameterName=Transition}
  ScalarParameter() {parameterName=EndTransition}
 ]
}
